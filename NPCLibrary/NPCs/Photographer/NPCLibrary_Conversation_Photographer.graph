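# Conversation database for the Photographer NPC Skye, defining what she says when a player talks to her.
# A single entry point leads to one speech node that picks a contextual greeting over a default line.
# Prioritized lines react to a first-ever meeting or specific player outfits; unprioritized lines are gated by per-match conversation count with this NPC.
# The speech then links out to the shared NPC Library quests and basic services conversations.
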ConversationDatabase /NPCLibrary/NPCs/Photographer/NPCLibrary_Conversation_Photographer
entry_points [{tag=Conversation.Entry.Photographer, destinations [#1]}]
nodes [
 ConversationEntryPointNode_7 {
  id=#1
  class=ConversationEntryPointNode
  outputs [#2]
  properties {entryTag {tagName=Conversation.Entry.Photographer}}
 }
 Speech_C_4 {
  id=#2
  class=Speech_C
  outputs [#3, #4]
  properties {
   nodeName=""
   generalConfig {
    defaultMessage="This Island is always an adventure!"
    contextualMessages [
     {
      message="Hi! I'm Skye. Nice to meet you!"
      contextRequirements [{requirement {class=FortControllerRequirement_IsFirstEverConversationWithNPC}}]
      priority {curve {rowName=Tandem.Speech.Priority.FirstEverInteraction}}
     }
     {
      message=I'm really glad we reconnected. I've missed my friends.
      contextRequirements [{requirement {
         class=FortControllerRequirement_HasCID
         properties {allowedCIds [
           &CID_693_Athena_Commando_M_BuffCat
           /BRCosmetics/Athena/Items/Cosmetics/Characters/CID_691_Athena_Commando_F_TNTina
           /BRCosmetics/Athena/Items/Cosmetics/Characters/CID_770_Athena_Commando_F_MechanicalEngineer
           &CID_692_Athena_Commando_M_HenchmanTough
          ]}
        }}]
      priority {curve {rowName=Tandem.Speech.Priority.OutfitReaction}}
     }
     {
      message=THANK YOU for finding him. I've been worried.
      contextRequirements [{requirement {
         class=FortControllerRequirement_HasCID
         properties {allowedCIds [/BRCosmetics/Athena/Items/Cosmetics/Characters/CID_A_034_Athena_Commando_F_CatBurglar]}
        }}]
      priority {curve {rowName=Tandem.Speech.Priority.OutfitReaction}}
     }
     {
      message="BOSS! You're back, safe and sound!"
      contextRequirements [{requirement {class=FortControllerRequirement_HasCID, properties {allowedCIds [/BRCosmetics/Athena/Items/Cosmetics/Characters/Character_PiteousKicks]}}}]
      priority {curve {rowName=Tandem.Speech.Priority.OutfitReaction}}
     }
     {
      message="This Island is always an adventure!"
      contextRequirements [{requirement {class=FortControllerRequirement_CheckNPCMatchConversationCount, properties {nPCUniqueIdTag {tagName=AISpawnerData.Type.Tandem.Photographer}}}}]
      priority {curve {rowName=None}}
     }
     {
      message="I think you'd like my friend Ollie - and he'd like you!"
      contextRequirements [{requirement {class=FortControllerRequirement_CheckNPCMatchConversationCount, properties {nPCUniqueIdTag {tagName=AISpawnerData.Type.Tandem.Photographer}}}}]
      priority {curve {rowName=None}}
     }
     {
      message="A new adventure is always on the horizon!"
      contextRequirements [{requirement {class=FortControllerRequirement_CheckNPCMatchConversationCount, properties {nPCUniqueIdTag {tagName=AISpawnerData.Type.Tandem.Photographer}}}}]
      priority {curve {rowName=None}}
     }
     {
      message="The old crew's back together, and it's... different, but nice."
      contextRequirements [{requirement {class=FortControllerRequirement_CheckNPCMatchConversationCount, properties {nPCUniqueIdTag {tagName=AISpawnerData.Type.Tandem.Photographer}}}}]
      priority {curve {rowName=None}}
     }
     {
      message=Never thought I'd get tangled back in the boss' webs.
      contextRequirements [{requirement {class=FortControllerRequirement_CheckNPCMatchConversationCount, properties {nPCUniqueIdTag {tagName=AISpawnerData.Type.Tandem.Photographer}}}}]
      priority {curve {rowName=None}}
     }
     {
      message=I'm so glad the boss is okay. I was worried about him.
      contextRequirements [{requirement {class=FortControllerRequirement_CheckNPCMatchConversationCount, properties {nPCUniqueIdTag {tagName=AISpawnerData.Type.Tandem.Photographer}}}}]
      priority {curve {rowName=None}}
     }
    ]
   }
  }
 }
 ConversationLinkNode_0 {id=#4, class=ConversationLinkNode, properties {remoteEntryTag {tagName=Conversation.Entry.BasicServices}}}
 ConversationLinkNode_1 {id=#3, class=ConversationLinkNode, properties {remoteEntryTag {tagName=Conversation.Entry.NPCLibrary.Quests}}}
]
